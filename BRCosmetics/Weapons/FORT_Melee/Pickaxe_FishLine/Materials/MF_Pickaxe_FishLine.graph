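# Material function for the FishLine pickaxe that layers an animated, icy liquid-blade look over incoming material attributes, rebuilding base color and emissive.
# A channel-selected TechArtMask defines the blade face, while pre-skinned position and bounds supply coordinates and edge lines.
# Flowmap-distorted bubbles and caustics, Perlin frost and parallax UVs build surface detail.
# Fresnel, GGX glint, fake subsurface scattering and refracted skylight, with a lat-long texture on forward and mobile, fake translucent lighting.
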MaterialFunction /BRCosmetics/Weapons/FORT_Melee/Pickaxe_FishLine/Materials/MF_Pickaxe_FishLine
graph {nodes [
  Constant2Vector() {g=2, r=2}
  Constant3Vector() {constant=<rgba 0 -0.125 0.1093 0>}
  n1 = GetMaterialAttributes(None=FunctionInput() {inputName=OUT_MA, inputType=FunctionInput_MaterialAttributes}) {attributeGetTypes [
    #1
    #2
    #3
    #4
    #5
    #6
   ]}
  n2 = SetMaterialAttributes!(
   MaterialAttributes=@n1.MaterialAttributes
   BaseColor=LinearInterpolate(A=NamedRerouteUsage(), B=NamedRerouteUsage(), Alpha=NamedRerouteUsage()) {constB=0}
   Metallic=@n1.MaterialAttributes
   Specular=@n1.MaterialAttributes
   Roughness=@n1.MaterialAttributes
   Normal=@n1.MaterialAttributes
   EmissiveColor=Multiply(A=NamedRerouteUsage(), B=NamedRerouteUsage())
  ) {attributeSetTypes [
    #1
    #2
    #3
    #4
    #5
    #6
   ]}
  FunctionOutput(None=@n2)
  n3 = ComponentMask(None=NamedRerouteUsage()) {b=true}
  Min(A=@n3, B=OneMinus(None=@n3))
  n4 = ScalarParameter() {defaultValue=1, group=FishLine, parameterName=RoughnessHighlightIntensity}
  Multiply(A=NamedRerouteUsage(), B=@n4)
  NamedRerouteDeclaration() {name=OUT_BaseColor, nodeColor=<rgba 0 0.8856 1 1>}
  NamedRerouteDeclaration() {name=OUT_Metallic, nodeColor=<rgba 0.3462 1 0.0006 1>}
  NamedRerouteDeclaration(None=@n1.EmissiveColor) {name=IN_EmissiveColor, nodeColor=<rgba 0 0.8 1 1>}
  n5 = Multiply(
   A=ScalarParameter() {defaultValue=0.2, parameterName=UVRadialScale}
   B=MaterialFunctionCall() {materialFunction=&ObjectScale}.Scale XYZ
  )
  n6 = MaterialFunctionCall() {materialFunction=/Engine/Functions/Engine_MaterialFunctions02/Utility/FOV}
  NamedRerouteDeclaration(None=MaterialFunctionCall!(Position=ObjectPositionWS(), UVRadialScale=Divide(A=@n5, B=@n6."FOV ")) {materialFunction=/Game/Characters/Player/Common/MaterialFunctions/Superlevel/CharacterStabilizedScreenSpaceUVs}.StabilizedUVs) {name=StablizedUV, nodeColor=<rgba 0 1 0.2031 1>}
  n7 = FunctionInput() {bUsePreviewValueAsDefault=true, inputName=TechArtMask, sortPriority=3, previewValue=<1 1 1 1>}
  NamedRerouteDeclaration(None=Step(X=ChannelMaskParameter(None=@n7) {parameterName=FishLine_FXMask}.R) {constX=0.1, constY=0.15}) {name=SwordFaceMask, nodeColor=<rgba 0.1016 1 0 1>}
  n8 = LocalPosition() {localOrigin=InstancePreSkinning}
  NamedRerouteDeclaration(None=ShaderStageSwitch(PixelShader=VertexInterpolator(VS=@n8.XYZ).PS, VertexShader=@n8.XYZ)) {name=PreSkinnedLoc, nodeColor=<rgba 0.7001 0 1 1>}
  n9 = NamedRerouteUsage()
  n10 = Multiply(A=@n9, B=ScalarParameter() {defaultValue=2, group=FishLine, parameterName=BubblesTile1})
  n11 = Multiply(A=Time(), B=ScalarParameter() {defaultValue=0.633, group=FishLine, parameterName=BubblesSpeed})
  n12 = ScalarParameter() {defaultValue=0.05, group=FishLine, parameterName=Bubble-FlowmapStrength1}
  n13 = ScalarParameter() {defaultValue=4, group=FishLine, parameterName=Bubble-FlowmapScale}
  n14 = Time()
  n15 = ScalarParameter() {defaultValue=0.2, group=FishLine, parameterName=Bubble-FlowmapSpeed}
  n16 = TextureSample(UVs=Panner(Coordinate=Multiply(A=@n9, B=@n13), Time=Multiply(A=@n14, B=@n15)) {speedX=1}) {
   samplerSource=SSM_Wrap_WorldGroupSettings
   samplerType=SAMPLERTYPE_LinearColor
   texture=/Game/Global/Textures/Default/Flowmaps/T_TorusFlowmap
  }
  n17 = ComponentMask(None=Multiply(A=Subtract(A=@n16.RGB) {constB=0.5}, B=VectorParameter() {group=FishLine, parameterName=Potion-FlowmapAmt, defaultValue=<rgba 1 1 0 1>}.RGB)) {g=true, r=true}
  n18 = TextureSample(UVs=Add(A=Panner(Coordinate=@n10, Time=@n11) {speedX=-0.1, speedY=-0.5}, B=Multiply(A=@n12, B=@n17))) {samplerSource=SSM_Wrap_WorldGroupSettings, samplerType=SAMPLERTYPE_Masks, texture=/Game/Global/Textures/Default/PointSplatter/Bubbles_M}
  n19 = Multiply(A=@n9, B=ScalarParameter() {defaultValue=4, group=FishLine, parameterName=BubblesTile2})
  n20 = ScalarParameter() {defaultValue=0.025, group=FishLine, parameterName=Bubble-FlowmapStrength2}
  n21 = TextureSample(UVs=Add(A=Panner(Coordinate=@n19, Time=@n11) {speedX=0.1, speedY=-0.5}, B=Multiply(A=@n17, B=@n20))) {samplerSource=SSM_Wrap_WorldGroupSettings, samplerType=SAMPLERTYPE_Masks, texture=&Bubbles_M}
  n22 = Max(A=@n18.RGB, B=@n21.RGB)
  n23 = OneMinus(None=@n22)
  n24 = ComponentMask(None=Multiply(A=Power(Base=@n23) {constExponent=4}, B=@n22)) {g=true}
  n25 = ScalarParameter() {defaultValue=4, group=FishLine, parameterName=BubbleMaskAmount}
  n26 = Add(A=ComponentMask(None=NamedRerouteUsage()) {b=true}) {constB=0.15}
  n27 = Power(Base=Min(A=@n26, B=OneMinus(None=@n26)))
  NamedRerouteDeclaration(None=Multiply(A=Multiply(A=@n24, B=@n25), B=Multiply(A=@n27) {constB=4})) {name=Bubble, nodeColor=<rgba 0 1 0.3203 1>}
  n28 = Multiply(A=CameraVectorWS()) {constB=-1}
  n29 = VertexNormalWS()
  n30 = MaterialFunctionCall(Float3=Subtract(
    A=Transform(None=@n28) {transformSourceType=TRANSFORMSOURCE_World, transformType=TRANSFORM_Tangent}
    B=Transform(None=@n29) {transformSourceType=TRANSFORMSOURCE_World, transformType=TRANSFORM_Tangent}
   )) {materialFunction=&BreakOutFloat3Components}
  n31 = MaterialFunctionCall(X=@n30.R, Y=@n30.R) {materialFunction=&MakeFloat2}
  n32 = Constant() {r=0.1}
  NamedRerouteDeclaration(None=Divide(A=Multiply(A=@n31.Result, B=@n32), B=Multiply(A=@n30.B) {constB=-1})) {name=ParallexOffset, nodeColor=<rgba 0.1953 1 0 1>}
  n33 = NamedRerouteUsage()
  n34 = Constant2Vector() {g=0.2, r=0.5}
  n35 = TextureSample(UVs=Panner(Coordinate=Multiply(A=@n33, B=@n34) {constB=0.2}) {speedY=-0.005}) {
   samplerSource=SSM_Wrap_WorldGroupSettings
   samplerType=SAMPLERTYPE_LinearColor
   texture=/Game/Global/Materials/MaterialGym/Materials/SimpleIce/T_Perlin_Noise_M1
  }
  n36 = Add(A=NamedRerouteUsage(), B=Multiply(A=@n35.R) {constB=0.5}) {constB=0}
  n37 = Saturate(None=@n36)
  NamedRerouteDeclaration(None=@n37) {name=Forset, nodeColor=<rgba 1 0 0.1641 1>}
  n38 = Multiply(A=ComponentMask(None=NamedRerouteUsage()) {b=true, r=true}, B=Constant2Vector() {g=2, r=0.5})
  n39 = TextureSample(UVs=Multiply(A=@n38, B=ScalarParameter() {defaultValue=0.1, group=FishLine, parameterName=UVScale}) {constB=2.5}) {
   samplerSource=SSM_Wrap_WorldGroupSettings
   samplerType=SAMPLERTYPE_LinearColor
   texture=&T_Perlin_Noise_M1
  }
  n40 = MaterialFunctionCall!(
   Normal=MaterialFunctionCall(Normal=NamedRerouteUsage(), Flatness=Constant() {r=0.5}) {materialFunction=&FlattenNormal}.Result
   CameraVector=Transform(None=CameraVectorWS()) {transformSourceType=TRANSFORMSOURCE_World, transformType=TRANSFORM_Tangent}
  ) {materialFunction=/Engine/Functions/Engine_MaterialFunctions02/Utility/CustomReflectionVector}
  n41 = ComponentMask(None=@n40.Reflection Vector) {b=true}
  n42 = Multiply(
   A=ComponentMask(None=@n40.Reflection Vector) {g=true, r=true}
   B=Divide(A=ScalarParameter() {defaultValue=10, group=FishLine, parameterName=InnerIceDepth}, B=Abs(None=@n41)) {constA=100}
  )
  n43 = Divide(B=Constant() {r=512}) {constB=512}
  NamedRerouteDeclaration(None=Add(A=BumpOffset(Coordinate=@n38, Height=@n39.R) {heightRatio=0.1}, B=Multiply(A=@n42, B=@n43))) {name=Crystaluv, nodeColor=<rgba 0.7578 1 0 1>}
  NamedRerouteDeclaration() {name=OUT_Specular, nodeColor=<rgba 0.25 1 0 1>}
  n44 = NamedRerouteUsage()
  n45 = Power(
   Base=Fresnel(BaseReflectFractionIn=Constant(), Normal=Transform(None=@n44)) {exponent=1}
   Exp=ScalarParameter() {defaultValue=1, group=FishLine, parameterName=HighlightFresnelExponent}
  ) {constExponent=1.2}
  n46 = ScalarParameter() {defaultValue=2, group=FishLine, parameterName=HighLightEmissive}
  NamedRerouteDeclaration(None=Multiply(A=@n36, B=Multiply(A=@n45, B=@n46))) {name=Edge highlight, nodeColor=<rgba 1 0 0.4922 1>}
  n47 = VertexNormalWS()
  n48 = CameraVectorWS()
  n49 = DotProduct(A=Normalize(VectorInput=Add(A=@n47, B=@n48)), B=Multiply(A=AtmosphericLightVector()) {constB=-1})
  n50 = VertexNormalWS()
  n51 = CameraVectorWS()
  n52 = Multiply(A=Add(A=Multiply(A=@n49) {constB=0.5}) {constB=0.5}, B=Saturate(None=DotProduct(A=@n50, B=@n51)))
  n53 = ScalarParameter() {defaultValue=4, group=FishLine, parameterName=FakeSSSExponent}
  NamedRerouteDeclaration(None=Saturate(None=Power(Base=@n52, Exp=@n53))) {name=FakeSSS, nodeColor=<rgba 0 0.0859 1 1>}
  NamedRerouteDeclaration(None=@n1.BaseColor) {name=IN_BaseColor, nodeColor=<rgba 0.875 1 0 1>}
  NamedRerouteDeclaration(None=@n1.Normal) {name=IN_Normal, nodeColor=<rgba 0.8047 0 1 1>}
  NamedRerouteDeclaration(None=@n1.Roughness) {name=IN_Roughness, nodeColor=<rgba 1 0.3516 0 1>}
  NamedRerouteDeclaration(None=@n1.Specular) {name=IN_Specular, nodeColor=<rgba 0.0312 0 1 1>}
  n54 = NamedRerouteUsage()
  n55 = VectorParameter() {
   group=FishLine
   parameterName=BladeEdgeLinePosition
   channelNames {a="", b=Edge_Horizon, g=Edge_Side, r=Edge_Vertical}
   defaultValue=<rgba 0.008 -0.01 0.1145 1>
  }
  n56 = Abs(None=ConstantBiasScale(None=Add(A=@n54, B=@n55.RGB)) {bias=-0.5, scale=2})
  n57 = NamedRerouteDeclaration(None=ComponentMask(None=@n56) {desc=Edge_Side, g=true}) {name=BladeEdge_Side, nodeColor=<rgba 1 0 0.6797 1>}
  n58 = SmoothStep(Value=@n57) {constMax=0.05}
  n59 = Multiply(A=NamedRerouteUsage(), B=@n58)
  n60 = SmoothStep(Value=@n57) {constMax=0, constMin=0.2}
  n61 = Transform(None=NamedRerouteUsage())
  n62 = Power(Base=OneMinus(None=Fresnel(Normal=@n61) {baseReflectFraction=0.05, exponent=0.1}))
  n63 = DotProduct(A=Min(A=@n58, B=Power(Base=@n60)), B=@n62)
  n64 = Multiply(A=NamedRerouteDeclaration(None=ComponentMask(None=@n56) {desc=Edge_Vertical, r=true}) {name=BladeEdge_V, nodeColor=<rgba 1 0 0.6797 1>}) {constB=0.25}
  n65 = NamedRerouteDeclaration(None=ComponentMask(None=@n56) {b=true, desc=Edge_Horizon_Main}) {name=BladeEdge_H, nodeColor=<rgba 1 0 0.6797 1>}
  n66 = VectorParameter() {
   group=FishLine
   parameterName=BladeEdgeLinePosition_Second
   channelNames {a="", b=Edge_Horizon, g=Edge_Side, r=Edge_Vertical}
   defaultValue=<rgba 0 0 0.4295 1>
  }
  n67 = ConstantBiasScale(None=Add(A=@n54, B=@n66.RGB)) {bias=-0.5, scale=2}
  n68 = NamedRerouteDeclaration(None=ComponentMask(None=Abs(None=@n67)) {b=true}) {name=BladeEdge_H2, nodeColor=<rgba 1 0 0.6797 1>}
  n69 = DotProduct(A=Add(A=@n64, B=Min(A=@n65, B=@n68)), B=@n62)
  n70 = Multiply(A=Max(A=Saturate(None=@n63), B=Saturate(None=@n69))) {constB=10}
  NamedRerouteDeclaration(None=Saturate(None=Max(A=@n59, B=@n70))) {name=SwordGraident, nodeColor=<rgba 0 1 0.0391 1>}
  NamedRerouteDeclaration(None=@n7) {name=InFXMask, nodeColor=<rgba 1 0.7969 0 1>}
  NamedRerouteDeclaration() {name=OUT_Roughness, nodeColor=<rgba 0.25 1 0 1>}
  n71 = CameraVectorWS()
  n72 = NamedRerouteUsage()
  n73 = Add(
   A=MaterialFunctionCall!(
    Ray Direction=Multiply(A=@n71) {constB=-1}
    Surface Normal=PixelNormalWS()
    Refractive Index Origin
    Refractive Index Target=ScalarParameter() {defaultValue=3, group=FishLine, parameterName=RefractionAmount}
   ) {materialFunction=/Engine/Functions/Engine_MaterialFunctions01/Vectors/Refract}.Vector
   B=Multiply(A=@n72) {constB=0.2}
  )
  n74 = TextureSample(UVs=MaterialFunctionCall!(Vector=ReflectionVectorWS()) {materialFunction=/Engine/Functions/Engine_MaterialFunctions01/Coordinates/LongLatToUV}.Result) {texture=/Game/Environments/Asteria/Sets/NeonCity/Techart/Textures/T_FakeInterior_03_Blur_LongLat}
  NamedRerouteDeclaration(None=ShadingPathSwitch(Default=SkyLightEnvMapSample(Direction=@n73), Forward=@n74.RGB, Mobile=@n74.RGB)) {name=EnvironmentMap, nodeColor=<rgba 0 1 0.0625 1>}
  n75 = VectorParameter() {group=FishLine, parameterName=Emissive_Color_01, defaultValue=<rgba 0.0525 0.1546 0.2292 1>}
  n76 = VectorParameter() {group=FishLine, parameterName=Emissive_Color_02, defaultValue=<rgba 0.2769 0.5591 0.901 1>}
  n77 = NamedRerouteUsage()
  n78 = Power(Base=VectorParameter() {group=FishLine, parameterName=B_Color, defaultValue=<rgba 0 0.5636 0.724 1>}.RGB)
  n79 = NamedRerouteUsage()
  n80 = Panner(Coordinate=Multiply(A=@n79, B=Constant2Vector() {g=0.2, r=1.5})) {speedY=-0.02}
  n81 = TextureSample(UVs=Panner(Coordinate=Multiply(A=@n79) {constB=4}) {speedX=0.05, speedY=-0.2}) {
   samplerSource=SSM_Wrap_WorldGroupSettings
   samplerType=SAMPLERTYPE_LinearColor
   texture=&T_TorusFlowmap
  }
  n82 = Multiply(A=ComponentMask(None=@n81.RGB) {g=true, r=true}, B=LinearInterpolate(Alpha=@n3) {constA=0.1, constB=0}) {constB=0.1}
  n83 = Power(Base=TextureSample(UVs=Add(A=@n80, B=@n82)) {samplerSource=SSM_Wrap_WorldGroupSettings, texture=/Game/Global/Textures/Default/Caustics/T_Caustics_Softer}.R)
  NamedRerouteDeclaration(None=LinearInterpolate(A=LinearInterpolate(A=@n75.RGB, B=@n76.RGB, Alpha=@n77), B=Multiply(A=@n78) {constB=5}, Alpha=Saturate(None=@n83))) {name=SwordEC, nodeColor=<rgba 0.5234 1 0 1>}
  n84 = TransformPosition(None=NamedRerouteUsage()) {transformType=TRANSFORMPOSSOURCE_Instance}
  n85 = PreSkinnedLocalBounds()
  NamedRerouteDeclaration(None=Divide(A=Subtract(A=@n84, B=@n85.Min), B=@n85.Extents)) {name=Bounding, nodeColor=<rgba 0 0.5312 1 1>}
  NamedRerouteDeclaration() {name=OUT_Normal, nodeColor=<rgba 0.25 0.25 1 1>}
  n86 = NamedRerouteUsage()
  n87 = ScalarParameter() {defaultValue=0.5, group=FishLine, parameterName=BaseColorEnvironmentAmount}
  n88 = NamedRerouteUsage()
  n89 = NamedRerouteUsage()
  n90 = ScalarParameter() {defaultValue=2, group=FishLine, parameterName=BaseColorForsetExp}
  n91 = Saturate(None=Multiply(A=Power(Base=@n89, Exp=@n90) {constExponent=1}, B=ScalarParameter() {defaultValue=1, group=FishLine, parameterName=BaseColorForsetInt}))
  NamedRerouteDeclaration(None=Multiply(A=Multiply(A=@n86, B=@n87) {constB=0.5}, B=Add(A=@n88, B=@n91))) {name=Base Color, nodeColor=<rgba 0.0781 1 0 1>}
  NamedRerouteDeclaration(None=Max(A=@n37, B=NamedRerouteUsage())) {name=IceHighlight, nodeColor=<rgba 0.2656 1 0 1>}
  n92 = SkyAtmosphereLightDirection()
  n93 = CameraVectorWS()
  n94 = MaterialFunctionCall!(Normal=PixelNormalWS(), Light Vector=Normalize(VectorInput=Add(A=@n92, B=@n93)), Roughness=Constant() {r=0.01}) {materialFunction=/Engine/Functions/Engine_MaterialFunctions01/Lighting/GGXSpecular}
  n95 = NamedRerouteUsage()
  n96 = Saturate(None=Multiply(A=Fresnel() {baseReflectFraction=0}) {constB=3})
  n97 = NamedRerouteUsage()
  n98 = NamedRerouteUsage()
  n99 = LinearInterpolate(A=Multiply(A=Add(A=@n95, B=@n96), B=@n97), B=Multiply(A=@n97, B=Multiply(A=@n98, B=@n4)), Alpha=NamedRerouteUsage()) {constAlpha=0.7}
  n100 = Multiply(A=@n99, B=VectorParameter() {group=FishLine, parameterName=FakeSSSColor, defaultValue=<rgba 2.5625 2.8656 3 1>}.RGB) {constB=3}
  n101 = NamedRerouteUsage()
  n102 = Constant() {r=0.8}
  n103 = Constant() {r=1}
  n104 = Constant() {r=0.6}
  n105 = Add(
   A=Saturate(None=Multiply(A=@n94.Result, B=@n94.Result))
   B=Multiply(
    A=LinearInterpolate(A=@n99, B=@n100, Alpha=@n101)
    B=MaterialFunctionCall(Color_Morning=@n102, Color_Day=@n103, Color_Evening=@n102, Color_Night=@n104) {materialFunction=&Color_Time_of_Day}.Result
   )
  )
  NamedRerouteDeclaration(None=Add(A=NamedRerouteUsage(), B=Add(A=@n105, B=@n100))) {name=Emissive Color, nodeColor=<rgba 1 0.3516 0 1>}
  NamedRerouteDeclaration() {name=OUT_EmissiveColor, nodeColor=<rgba 0 0.8 1 1>}
  n106 = NamedRerouteUsage()
  n107 = Power(Base=@n106)
  n108 = Divide(A=Add(A=@n64, B=@n57)) {constB=0.1}
  n109 = Divide(A=@n108, B=NamedRerouteUsage()) {constA=2}
  n110 = Divide(A=@n108, B=NamedRerouteUsage())
  n111 = NamedRerouteUsage()
  n112 = Max(A=@n106, B=Max(A=Max(A=@n109, B=@n110), B=Divide(A=@n108, B=@n111) {constA=0.001}))
  n113 = Saturate(None=Multiply(A=@n62, B=Multiply(A=@n107, B=@n112)))
  n114 = TextureSample(UVs=Add(A=Multiply(A=@n80) {constB=1.5}, B=@n82) {constB=0.25}) {samplerSource=SSM_Wrap_WorldGroupSettings, texture=&T_Caustics_Softer}
  n115 = Multiply(A=@n70, B=Saturate(None=Power(Base=@n114.R)))
  NamedRerouteDeclaration(None=Multiply(A=Max(A=@n113, B=@n115)) {constB=0.8}) {name=BladeEdgeLight, nodeColor=<rgba 0 1 0.4141 1>}
  NamedRerouteDeclaration(None=@n1.Metallic) {name=IN_Metallic, nodeColor=<rgba 0.25 1 0 1>}
  NamedRerouteUsage()
  NamedRerouteUsage()
  NamedRerouteUsage()
  NamedRerouteUsage()
  NamedRerouteUsage()
  NamedRerouteUsage()
  Subtract(A=@n72) {constB=0}
 ]}
